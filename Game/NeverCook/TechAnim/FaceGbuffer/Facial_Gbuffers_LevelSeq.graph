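# Level Sequence under NeverCook/TechAnim that frames two male soldier mannequins in a head-height shot, apparently for facial G-buffer work.
# A static camera with narrowed FOV feeds one camera cut, facing the mannequins placed side by side with opposite slight yaws.
# Both carry matching layouts of stacked skeletal animation tracks, one chaining six back-to-back clips.
# Level visibility sections and an event trigger track accompany the shot.
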LevelSequence /Game/NeverCook/TechAnim/FaceGbuffer/Facial_Gbuffers_LevelSeq
bindings [
 CameraActor {
  id=#1
  possessed_class=CameraActor
  tracks [{
    type=MovieScene3DTransformTrack
    sections [{
      type=MovieScene3DTransformSection
      channels [
       Location.X {type=MovieSceneScriptingDoubleChannel, default=-13}
       Location.Y {type=MovieSceneScriptingDoubleChannel, default=100}
       Location.Z {type=MovieSceneScriptingDoubleChannel, default=171}
       Rotation.X {type=MovieSceneScriptingDoubleChannel, default=0}
       Rotation.Y {type=MovieSceneScriptingDoubleChannel, default=-9.5}
       Rotation.Z {type=MovieSceneScriptingDoubleChannel, default=-90.2}
       Scale.X {type=MovieSceneScriptingDoubleChannel, default=1}
       Scale.Y {type=MovieSceneScriptingDoubleChannel, default=1}
       Scale.Z {type=MovieSceneScriptingDoubleChannel, default=1}
      ]
     }]
   }]
 }
 CameraComponent {
  id=#2
  parent_id=#1
  possessed_class=CameraComponent
  tracks [{
    type=MovieSceneFloatTrack
    display_name=Field Of View
    sections [{type=MovieSceneFloatSection, channels [None {type=MovieSceneScriptingFloatChannel, default=35}]}]
   }]
 }
 SK_Mannequin2 {
  id=#3
  possessed_class=SoldierMannequin_Male_C
  tracks [
   {
    type=MovieScene3DTransformTrack
    sections [{
      type=MovieScene3DTransformSection
      channels [
       Location.X {type=MovieSceneScriptingDoubleChannel, default=-30}
       Location.Y {type=MovieSceneScriptingDoubleChannel, default=0}
       Location.Z {type=MovieSceneScriptingDoubleChannel, default=0}
       Rotation.X {type=MovieSceneScriptingDoubleChannel, default=0}
       Rotation.Y {type=MovieSceneScriptingDoubleChannel, default=0}
       Rotation.Z {type=MovieSceneScriptingDoubleChannel, default=-5}
       Scale.X {type=MovieSceneScriptingDoubleChannel, default=1}
       Scale.Y {type=MovieSceneScriptingDoubleChannel, default=1}
       Scale.Z {type=MovieSceneScriptingDoubleChannel, default=1}
      ]
     }]
   }
   {
    type=MovieSceneSkeletalAnimationTrack
    sections [
     {type=MovieSceneSkeletalAnimationSection, end_frame=215, start_frame=0}
     {type=MovieSceneSkeletalAnimationSection, end_frame=393, start_frame=215}
     {type=MovieSceneSkeletalAnimationSection, end_frame=543, start_frame=393}
     {type=MovieSceneSkeletalAnimationSection, end_frame=922, start_frame=543}
     {type=MovieSceneSkeletalAnimationSection, end_frame=1212, start_frame=922}
     {type=MovieSceneSkeletalAnimationSection, end_frame=1698, start_frame=1212}
    ]
   }
   {type=MovieSceneSkeletalAnimationTrack, sections [{type=MovieSceneSkeletalAnimationSection, end_frame=2200, start_frame=0}]}
   {type=MovieSceneSkeletalAnimationTrack, sections [{type=MovieSceneSkeletalAnimationSection, end_frame=1900, start_frame=0}]}
   {type=MovieSceneSkeletalAnimationTrack, sections [{type=MovieSceneSkeletalAnimationSection, end_frame=1359, start_frame=0}]}
   {type=MovieSceneSkeletalAnimationTrack, sections [{type=MovieSceneSkeletalAnimationSection, end_frame=2666, start_frame=0}]}
  ]
 }
 SK_Mannequin {
  id=#4
  possessed_class=SoldierMannequin_Male_C
  tracks [
   {
    type=MovieSceneSkeletalAnimationTrack
    sections [
     {type=MovieSceneSkeletalAnimationSection, end_frame=215, start_frame=0}
     {type=MovieSceneSkeletalAnimationSection, end_frame=393, start_frame=215}
     {type=MovieSceneSkeletalAnimationSection, end_frame=543, start_frame=393}
     {type=MovieSceneSkeletalAnimationSection, end_frame=922, start_frame=543}
     {type=MovieSceneSkeletalAnimationSection, end_frame=1212, start_frame=922}
     {type=MovieSceneSkeletalAnimationSection, end_frame=1698, start_frame=1212}
    ]
   }
   {
    type=MovieScene3DTransformTrack
    sections [{
      type=MovieScene3DTransformSection
      channels [
       Location.X {type=MovieSceneScriptingDoubleChannel, default=0}
       Location.Y {type=MovieSceneScriptingDoubleChannel, default=0}
       Location.Z {type=MovieSceneScriptingDoubleChannel, default=0}
       Rotation.X {type=MovieSceneScriptingDoubleChannel, default=0}
       Rotation.Y {type=MovieSceneScriptingDoubleChannel, default=0}
       Rotation.Z {type=MovieSceneScriptingDoubleChannel, default=5}
       Scale.X {type=MovieSceneScriptingDoubleChannel, default=1}
       Scale.Y {type=MovieSceneScriptingDoubleChannel, default=1}
       Scale.Z {type=MovieSceneScriptingDoubleChannel, default=1}
      ]
     }]
   }
   {type=MovieSceneSkeletalAnimationTrack, sections [{type=MovieSceneSkeletalAnimationSection, end_frame=2200, start_frame=0}]}
   {type=MovieSceneSkeletalAnimationTrack, sections [{type=MovieSceneSkeletalAnimationSection, end_frame=1900, start_frame=0}]}
   {type=MovieSceneSkeletalAnimationTrack, sections [{type=MovieSceneSkeletalAnimationSection, end_frame=2666, start_frame=0}]}
   {type=MovieSceneSkeletalAnimationTrack, sections [{type=MovieSceneSkeletalAnimationSection, end_frame=1359, start_frame=0}]}
  ]
 }
]
display_rate {denominator=1, numerator=30}
tick_resolution {denominator=1, numerator=24000}
tracks [
 {
  type=MovieSceneLevelVisibilityTrack
  sections [
   {type=MovieSceneLevelVisibilitySection, end_frame=2817, start_frame=0}
   {type=MovieSceneLevelVisibilitySection, end_frame=2819, start_frame=0}
  ]
 }
 {
  type=MovieSceneEventTrack
  sections [{
    type=MovieSceneEventTriggerSection
    channels [None {type=MovieSceneScriptingEventChannel, const ["<Struct 'MovieSceneEvent' (0x49e5f37a3c0) {}>", 1]}]
   }]
 }
 {type=MovieSceneCameraCutTrack, sections [{type=MovieSceneCameraCutSection, end_frame=2825, start_frame=0}]}
]
